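# Sound cue for the partial-state sound of the Pipe Puzzle door, an interactable in the Hermes content.
# A single wave plays through a modulator at default settings into a one-input mixer, with boosted volume on the Fort_Quests sound class.
# It borrows the HiddenStar ambient loop attenuation from Athena supply drops.
# A concurrency override limits it to one instance per owner and rejects new plays while one is active.
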SoundCue /Hermes/Audio/Pipe_Puzzle/Door/Interactables_S34_Pipe_Puzzle_Door_Partial_Cue
root=Mixer_0
nodes [
 {id=Mixer_0, children [Modulator_0], properties {inputVolume [1]}}
 {id=Modulator_0, children [WavePlayer_0]}
 {id=WavePlayer_0, properties {soundWaveAssetPtr=&/Interactables_S34_Pipe_Puzzle_Door_Partial_01}}
]
properties {
 attenuationSettings=/Game/Athena/Sounds/SupplyDrops/Fort_HiddenStar_Ambient_Loop_Att
 bOverrideConcurrency=true
 maxDistance=2500
 soundClassObject=&Fort_Quests
 volumeMultiplier=3
 concurrencyOverrides {bLimitToOwner=true, maxCount=1, resolutionRule=PreventNew}
}
